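# Physics asset for the universal cowboy hat back accessory, filed under the cape cosmetics as a medium female variant.
# It has a single body on the backpack_root bone, made of one capsule that is offset from the bone and pitched 90 degrees.
# There are no constraints, so the asset defines one simple collision volume with no jointed chain.
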PhysicsAsset /BRCosmetics/Accessories/FORT_Capes/Cape_F_MED_CowboyHatUniversal/Meshes/F_MED_CowboyHatUniversal_Cape_Physics
bodies {backpack_root [{shapeType=Capsule, length=15, radius=15, shapeName=backpack_root_capsule, center=<0 -23.7507 -6.5442>, rotation=<rot 90 0 0>}]}
constraints []
